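# Niagara boost trail for the DragoncatLTE vehicle cosmetic, built from a single ribbon emitter with three ribbon renderers.
# Shared VC modules filter sockets, handle spawn rate by quality, and position one ribbon per port from the skeletal mesh plus a user car offset.
# Particles stream along negative X and widen over life, with narrower ribbons in frontend preview.
# Fixed yellow and orange colors are written as particle parameters, and the painted-variant override is disabled.
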NiagaraSystem /VehicleCosmetics/Boosts/Boost_DragoncatLTE/FX/NS_Boost_DragoncatLTE
dependencies {
 usedDataInterfaces [&NiagaraDataInterfaceCurve, &NiagaraDataInterfaceSkeletalMesh]
 usedDynamicInputs []
 usedModules [
  &self:SystemSpawnScript.SystemScriptSource.SystemScriptGraph.NiagaraNodeAssignment_0.SetVariables_#1
  &SystemState
 ]
}
emitters {NE_VC_Boost_Ribbon_Parent {
  emitterUpdateScript [
   &EmitterState()
   &VC_FilteredSocketCount()
   &VC_RateQualityHandler()
   &VC_SpawnRibbon_v1(Number of Ribbons=$Emitter.PortCount, SpawnRate=$Emitter.Spawn Rate)
  ]
  particleSpawnScript [
   &VC_PaintedVariantsOverride() {enabled=false}
   &VC_InitializeRibbon_v1(
    Custom Offset=true
    Enable Skeletal Position=true
    Lifetime=0.8
    Local Vector Offset=&AddVectorToPosition(Position=<0 0 0>, Vector=$User.DTCarOffset)
    Offset Amount=1
    Ribbon Width=80
   )
   &AddVelocity(Velocity=<-1400 0 0>) {module=AddVelocity001}
   &Color()
   /SetVariables(Particles.Amarillo:LinearColor=<rgba 1 0.3391 0 4>, Particles.Naranja:LinearColor=<rgba 1.1139 0.1304 0 1>) {isSetParameters=true}
  ]
  particleUpdateScript [
   &ParticleState()
   &SolveForcesAndVelocity(Clamp Velocity=false, Limit Acceleration=false)
   &VC_UpdateRibbon_v1(Enable Skeletal Position=true)
   &ScaleColor(Scale RGB=&VectorFromFloat(Value=1), ScaleA=true, ScaleRGB=true)
   &ScaleRibbonWidth(Ribbon Width Scale=&FloatFromCurve(
     FloatCurve=DataInterface() {propertyValues {Curve {keys [<at 0 0.75>, {arriveTangentWeight=0.3333, interpMode=RCIM_Cubic, tangentMode=RCTM_User, time=1, value=1.4}]}}}
     Scale Curve=&MakeCustomFloatFromBool(Bool=$User.bIsFrontendPreview, True Float=0.8, False Float=1)
    ))
  ]
  rendererClasses [&NiagaraRibbonRendererProperties]
  renderers [
   {rendererClass=&NiagaraRibbonRendererProperties}
   {rendererClass=&NiagaraRibbonRendererProperties, rendererIndex=1}
   {rendererClass=&NiagaraRibbonRendererProperties, rendererIndex=2}
  ]
 }}
summary {
 systemName=NS_Boost_DragoncatLTE
 emitters [{emitterName=NE_VC_Boost_Ribbon_Parent, rendererClasses [&NiagaraRibbonRendererProperties]}]
 userVariables [
  User.bIsFrontendPreview {type=NiagaraBool}
  User.bIsPainted {type=NiagaraBool}
  User.ColorVariantID {type=NiagaraInt32}
  User.DTCarOffset {type=Vector3f, description="", defaultValue {struct=Vector3f, value=<0 0 0>}}
  User.PaintedColor {type=LinearColor, description="", defaultValue {struct=LinearColor, value=<rgba 1 1 1 1>}}
  User.PaintedScalar {type=NiagaraFloat, defaultValue=1, description=""}
  User.PortCount {type=NiagaraInt32, defaultValue=1}
 ]
}
